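# Niagara skydiving contrail for the Underground Rebel cosmetic, trailing multicolored paint-like ribbons and droplets from the player's hands and feet.
# Five colored hand ribbon emitters and a purple feet one share a system paint velocity, with differing accelerations separating colors.
# Ribbons drive dynamic material parameters and scale alpha by time of day.
# GPU sprite droplets respond to glide lean and front-end sizing; the FortniteLetters sprite emitter is disabled.
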NiagaraSystem /BRCosmetics/Athena/Effects/Particles/Skydiving/NS_Contrail_UndergroundRebel/NS_Contrail_UnderGroundRebel
dependencies {
 usedDataInterfaces [&NiagaraDataInterfaceColorCurve, &NiagaraDataInterfaceCurve]
 usedDynamicInputs [&SimulationPosition, &FN_FrontEndVector3Modifier]
 usedModules [
  &self:SystemSpawnScript.SystemScriptSource.SystemScriptGraph.NiagaraNodeAssignment_2.SetVariables_#1
  &self:SystemSpawnScript.SystemScriptSource.SystemScriptGraph.NiagaraNodeAssignment_3.SetVariables_#2
  &SystemState
  &self:SystemSpawnScript.SystemScriptSource.SystemScriptGraph.NiagaraNodeAssignment_0.SetVariables_#3
  &self:SystemSpawnScript.SystemScriptSource.SystemScriptGraph.NiagaraNodeAssignment_1.SetVariables_#4
 ]
}
emitters {
 DropletsFeet {
  simTarget=GPUComputeSim
  emitterUpdateScript [
   &EmitterState(Inactive Response=::NewEnumerator1, Life Cycle Mode=::NewEnumerator1, Loop Behavior=::NewEnumerator1, Loop Duration=3600, UseLoopDelay=false)
   &SpawnRate(
    SpawnGroup:NiagaraInt32=&RandomRangeInt(Minimum=0, Maximum=1)
    SpawnRate=15
    Use Spawn Probability=false
   )
  ]
  particleSpawnScript [
   &FN_SetSpawnGroup(Number of Spawn Groups=2)
   &InitializeParticle(
    Color=&MakeCustomLinearColorFromBool(
     Bool=&RandomBool()
     True Color=&MakeCustomLinearColorFromBool(Bool=&RandomBool(), True Color=<rgba 0.2865 0 0.086 1>, False Color=<rgba 0 0.138 0.276 1>)
     False Color=&MakeCustomLinearColorFromBool(Bool=&RandomBool(), True Color=<rgba 5 1.1051 0 1>, False Color=<rgba 1 1 1 1>)
    )
    Lifetime Max=0.2
    Lifetime Min=0.125
    Lifetime Mode=::NewEnumerator1
    UsePositionOffset=false
    Write Material Random=true
   ) {module=InitializeParticle001}
   &SkeletalMeshLocation(
    Bone / Socket Index=$Particles.SpawnGroup
    Bone Sampling Mode=::NewEnumerator4
    Interpolate Mesh Sampling=false
    MaskSpawnGroup=false
    Skeletal Mesh=$System.SkelMeshFeet
   )
   /SetVariables(Particles.SpriteSize=&Multiply_Vector2(
     A=<0.5 1>
     B=&Multiply_Vector2D_ByFloat(
      Float=&FN_FrontEndFloatModifier(Gameplay=1, Front End=0.5)
      Vector2D=&Vector2DFromFloat(Value=&UniformRangedFloat(Minimum=1, Maximum=2))
     )
    )) {isSetParameters=true, module=SetVariables_#5}
   /SetVariables(Particles.Velocity=&MakeVector(
     X=&RandomRangeFloat(Minimum=-100, Maximum=100)
     Y=&RandomRangeFloat(Minimum=-100, Maximum=100)
     Z=&Multiply_Float(
      A=&RandomRangeFloat(Minimum=250, Maximum=400)
      B=&Lerp_Float(A=1, B=2, Alpha=$User.ParaGlideLeanAlpha)
     )
    )) {isSetParameters=true}
  ]
  particleUpdateScript [
   &ParticleState()
   &GravityForce(Gravity=<0 0 -100>)
   &Drag(Use Linear Drag=true, Use Rotational Drag=false)
   &SolveForcesAndVelocity(Clamp Velocity=false, Limit Acceleration=false)
   &ScaleColor(
    Scale Alpha=&FloatFromCurve(FloatCurve=DataInterface() {propertyValues {Curve {keys [
         <at 0 0 RCIM_Cubic>
         {arriveTangent=1.6041, leaveTangent=1.6041, time=0.05, value=1.0113}
         {interpMode=RCIM_Cubic, leaveTangent=-1.0633, time=0.75, value=1}
         <at 0.99 0 RCIM_Cubic>
        ]}}})
    Scale RGB=<0.6 0.8 1>
    ScaleA=true
    ScaleRGB=false
   )
   &ScaleSpriteSize(Scale Factor=&Vector2DFromFloat(Value=&FloatFromCurve(
      FloatCurve=DataInterface() {propertyValues {Curve {keys [
          {arriveTangent=-0.4919, interpMode=RCIM_Cubic, leaveTangent=-0.4919, tangentMode=RCTM_User, time=0.25, value=2}
          {arriveTangent=-3.3301, interpMode=RCIM_Cubic, leaveTangent=-3.3302, tangentMode=RCTM_User, time=1, value=0}
         ]}}}
      Scale Curve=6
     )))
  ]
  rendererClasses [&NiagaraSpriteRendererProperties]
  renderers [{rendererClass=&NiagaraSpriteRendererProperties}]
 }
 DropletsHands {
  simTarget=GPUComputeSim
  emitterUpdateScript [
   &EmitterState(Inactive Response=::NewEnumerator1, Life Cycle Mode=::NewEnumerator1, Loop Behavior=::NewEnumerator1, Loop Duration=3600, UseLoopDelay=false)
   &SpawnRate(
    SpawnGroup:NiagaraInt32=&RandomRangeInt(Minimum=0, Maximum=1)
    SpawnRate=25
    Use Spawn Probability=false
   )
  ]
  particleSpawnScript [
   &FN_SetSpawnGroup(Number of Spawn Groups=2)
   &InitializeParticle(
    Color=&MakeCustomLinearColorFromBool(
     Bool=&RandomBool()
     True Color=&MakeCustomLinearColorFromBool(Bool=&RandomBool(), True Color=<rgba 0.2865 0 0.086 1>, False Color=<rgba 0 0.138 0.276 1>)
     False Color=&MakeCustomLinearColorFromBool(Bool=&RandomBool(), True Color=<rgba 5 1.1051 0 1>, False Color=<rgba 1 1 1 1>)
    )
    Lifetime Max=0.75
    Lifetime Min=0.5
    Lifetime Mode=::NewEnumerator1
    UsePositionOffset=false
    Write Material Random=true
   ) {module=InitializeParticle001}
   &SkeletalMeshLocation(
    Bone / Socket Index=$Particles.SpawnGroup
    Bone Sampling Mode=::NewEnumerator4
    Interpolate Mesh Sampling=false
    MaskSpawnGroup=false
    Skeletal Mesh=$System.Skeletal Mesh
   )
   /SetVariables(Particles.Velocity=&MakeVector(
     X=&RandomRangeFloat(Minimum=-100, Maximum=100)
     Y=&RandomRangeFloat(Minimum=-100, Maximum=100)
     Z=&Multiply_Float(
      A=&RandomRangeFloat(Minimum=250, Maximum=400)
      B=&Lerp_Float(A=1, B=2, Alpha=$User.ParaGlideLeanAlpha)
     )
    )) {isSetParameters=true}
   /SetVariables(Particles.SpriteSize=&Multiply_Vector2(
     A=<5 10>
     B=&Multiply_Vector2D_ByFloat(
      Float=&FN_FrontEndFloatModifier(Gameplay=1, Front End=0.5)
      Vector2D=&Vector2DFromFloat(Value=&UniformRangedFloat(Minimum=1, Maximum=2))
     )
    )) {isSetParameters=true, module=SetVariables_#5}
  ]
  particleUpdateScript [
   &ParticleState()
   &GravityForce(Gravity=<0 0 -100>)
   &Drag(Use Linear Drag=true, Use Rotational Drag=false)
   &CurlNoiseForce(Noise Frequency=2, Noise Strength=&Lerp_Float(A=100, B=200, Alpha=$User.ParaGlideLeanAlpha), Pan Noise=false)
   &SolveForcesAndVelocity(Clamp Velocity=false, Limit Acceleration=false)
   &ScaleColor(
    Scale Alpha=&FloatFromCurve(FloatCurve=DataInterface() {propertyValues {Curve {keys [
         <at 0 0 RCIM_Cubic>
         {arriveTangent=1.6041, leaveTangent=1.6041, time=0.05, value=1.0113}
         {interpMode=RCIM_Cubic, leaveTangent=-1.0633, time=0.75, value=1}
         <at 0.99 0 RCIM_Cubic>
        ]}}})
    Scale RGB=<0.6 0.8 1>
    ScaleA=true
    ScaleRGB=false
   )
   &ScaleSpriteSize(Scale Factor=&Vector2DFromFloat(Value=&FloatFromCurve(
      FloatCurve=DataInterface() {propertyValues {Curve {keys [
          {arriveTangent=-0.4919, interpMode=RCIM_Cubic, leaveTangent=-0.4919, tangentMode=RCTM_User, time=0.25, value=2}
          {arriveTangent=-3.3301, interpMode=RCIM_Cubic, leaveTangent=-3.3302, tangentMode=RCTM_User, time=1, value=0}
         ]}}}
      Scale Curve=0.65
     )))
  ]
  rendererClasses [&NiagaraSpriteRendererProperties]
  renderers [{rendererClass=&NiagaraSpriteRendererProperties}]
 }
 FortniteLetters {
  bEnabled=false
  emitterSpawnScript [&FN_PreserveSkydiveUserVariables() {module=PreserveUserVariables}]
  emitterUpdateScript [
   &EmitterState(
    Life Cycle Mode=::NewEnumerator1
    Loop Delay=2
    Loop Duration=&Multiply_Float(A=8, B=$System.SpawnTimeOffset)
    UseLoopDelay=true
   )
   &SpawnBurst_Instantaneous(
    Spawn Time=&Multiply_Float(A=7, B=$System.SpawnTimeOffset)
    Use Spawn Probability=false
    UseLoopCountLimit=false
   )
   &SpawnBurst_Instantaneous(
    Spawn Time=&Multiply_Float(A=6, B=$System.SpawnTimeOffset)
    Use Spawn Probability=false
    UseLoopCountLimit=false
   ) {module=SpawnBurst_Instantaneous001}
   &SpawnBurst_Instantaneous(
    Spawn Time=&Multiply_Float(A=5, B=$System.SpawnTimeOffset)
    Use Spawn Probability=false
    UseLoopCountLimit=false
   ) {module=SpawnBurst_Instantaneous002}
   &SpawnBurst_Instantaneous(
    Spawn Time=&Multiply_Float(A=4, B=$System.SpawnTimeOffset)
    Use Spawn Probability=false
    UseLoopCountLimit=false
   ) {module=SpawnBurst_Instantaneous003}
   &SpawnBurst_Instantaneous(
    Spawn Time=&Multiply_Float(A=3, B=$System.SpawnTimeOffset)
    Use Spawn Probability=false
    UseLoopCountLimit=false
   ) {module=SpawnBurst_Instantaneous004}
   &SpawnBurst_Instantaneous(
    Spawn Time=&Multiply_Float(A=2, B=$System.SpawnTimeOffset)
    Use Spawn Probability=true
    UseLoopCountLimit=false
   ) {module=SpawnBurst_Instantaneous005}
   &SpawnBurst_Instantaneous(
    Spawn Time=&Multiply_Float(A=1, B=$System.SpawnTimeOffset)
    Use Spawn Probability=true
    UseLoopCountLimit=false
   ) {module=SpawnBurst_Instantaneous006}
   &SpawnBurst_Instantaneous(
    Spawn Time=&Multiply_Float(A=0, B=$System.SpawnTimeOffset)
    Use Spawn Probability=true
    UseLoopCountLimit=false
   ) {module=SpawnBurst_Instantaneous007}
  ]
  particleSpawnScript [
   &PartitionParticles(Number Of Partitions=8, Particles Per Partition=1, Partition Mode:ENiagara_PartitionMode=::NewEnumerator1)
   &InitializeParticle(
    Color=<rgba 1 1 1 10>
    Mass Mode=::NewEnumerator1
    Sprite Rotation Angle=&RandomRangeFloat(Minimum=-25, Maximum=25)
    Sprite Rotation Mode=::NewEnumerator0
    Sprite Size Mode=::NewEnumerator1
    Uniform Sprite Size Max=50
    Uniform Sprite Size Min=30
    UsePositionOffset=false
    Write Material Random=true
   ) {module=InitializeParticle001}
   /SetVariables(Particles.PartitionPerLetter:NiagaraInt32=$Particles.PartitionParticles.Partition) {isSetParameters=true}
   &PartitionParticles(Partition Mode:ENiagara_PartitionMode=::NewEnumerator1) {module=PartitionParticles001}
   /SetVariables(Particles.PartitionPerArm:NiagaraInt32=$Particles.PartitionParticles001.Partition) {isSetParameters=true}
   &SkeletalMeshLocation(
    Bone / Socket Index=$Particles.PartitionPerArm
    Bone Sampling Mode=::NewEnumerator4
    MaskSpawnGroup=false
    Sampled Position Coordinate Space=::Local
    Sampled Position Offset=<0 0 25>
    Sampled Position Offset Coordinate Space:ENiagaraSkelMeshTransforms=::NewEnumerator1
    Skeletal Mesh=$System.Skeletal Mesh
   )
   /SetVariables(Particles.SubImageIndex=&MakeFloatFromInt(Int=$Particles.PartitionPerLetter)) {isSetParameters=true}
   &AddVelocity(
    Velocity=<0 0 1>
    Velocity Speed Scale=&Multiply_Float(A=600, B=&Lerp_Float(A=1, B=1.5, Alpha=$User.ParaGlideLeanAlpha))
   )
  ]
  particleUpdateScript [
   /Niagara/Modules/Update/Forces/CurlNoiseForce(
    Noise Strength=&FN_FrontEndFloatModifier(
     Gameplay=100
     Front End=&UniformRangedFloat(
      Minimum=&FloatFromCurve(FloatCurve=DataInterface() {propertyValues {Curve {keys [<at 0.5 0>,<at 1.0011 1000>]}}})
      Maximum=&FloatFromCurve(FloatCurve=DataInterface() {propertyValues {Curve {keys [<at 0.5 0>,<at 1 -1000>]}}})
     )
    )
    Pan Noise=false
   )
   &ParticleState()
   &Drag(
    Drag=&FloatFromCurve(FloatCurve=DataInterface() {propertyValues {Curve {keys [<at 0 0>,<at 1 1>]}}}, Scale Curve=50)
    Use Linear Drag=true
    Use Rotational Drag=true
   )
   &SolveForcesAndVelocity(Clamp Velocity=false, Limit Acceleration=false)
   &ScaleSpriteSize(Scale Factor=&Vector2DFromFloat(Value=&FloatFromCurve(
      FloatCurve=DataInterface() {propertyValues {Curve {keys [
          {arriveTangent=9.0725, interpMode=RCIM_Cubic, leaveTangent=9.0725, tangentMode=RCTM_User, time=0, value=0.1}
          {arriveTangent=-1.0465, interpMode=RCIM_Cubic, leaveTangent=-1.0465, tangentMode=RCTM_User, time=0.1408, value=0.9847}
          {arriveTangent=-0.0405, interpMode=RCIM_Cubic, leaveTangent=-0.0405, tangentMode=RCTM_User, time=1, value=0.5}
         ]}}}
      Scale Curve=2
     )))
   &ScaleColor(
    Scale Alpha=&MakeFloatFromVector2d(Vector 2D=&Vector2DFromFloat(Value=&FloatFromCurve(FloatCurve=DataInterface() {propertyValues {Curve {keys [
           {arriveTangent=9.0725, interpMode=RCIM_Cubic, leaveTangent=9.0725, tangentMode=RCTM_User, time=0, value=0}
           {arriveTangent=-2.5249, interpMode=RCIM_Cubic, leaveTangent=-2.5249, tangentMode=RCTM_User, time=0.1408, value=0.9847}
           {arriveTangent=-0.0405, interpMode=RCIM_Cubic, leaveTangent=-0.0405, tangentMode=RCTM_User, time=1, value=0}
          ]}}})))
    ScaleA=true
    ScaleRGB=false
   )
   &CameraOffset(Camera Offset Amount=10)
  ]
  rendererClasses [&NiagaraSpriteRendererProperties]
  renderers [{rendererClass=&NiagaraSpriteRendererProperties}]
 }
 RibbonBlue {
  emitterUpdateScript [
   &EmitterState()
   /Game/Effects/Niagara/Modules/Ribbons/FN_SpawnRibbon_v3(Number of Ribbons=2, Skeleton Mesh:NiagaraDataInterfaceSkeletalMesh=$System.Skeletal Mesh, SpawnRate=7)
  ]
  particleSpawnScript [
   &FN_SetSpawnGroup(Number of Spawn Groups=2)
   /Game/Effects/Niagara/Modules/Ribbons/FN_InitializeRibbon_v5(
    Enable Skeletal Position=true
    Lifetime:NiagaraFloat=&Multiply_Float(A=0.84, B=$System.RibbonHandScalar)
    Ribbon Width=50
    SkeletalMesh:NiagaraDataInterfaceSkeletalMesh=$System.Skeletal Mesh
   )
   &AddVelocity(Velocity=$System.PaintVelocity, Velocity Speed Scale=$System.PaintVelocityScale) {module=AddVelocity001}
   /SetVariables(Particles.Color=&ColorFromCurve(
     ColorCurve=DataInterface() {propertyValues {
       AlphaCurve {keys [<at 0 1 RCIM_Cubic>]}
       BlueCurve {keys [<at 0 0.276>,<at 0.3701 0.6198 RCIM_Cubic>]}
       GreenCurve {keys [<at 0 0.138>,<at 0.3701 0.0413 RCIM_Cubic>]}
       RedCurve {keys [<at 0 0>,<at 0.3701 0 RCIM_Cubic>]}
      }}
     CurveIndex=&RandomRangeFloat(Minimum=0, Maximum=1)
    )) {isSetParameters=true, module=SetVariables_#6}
  ]
  particleUpdateScript [
   &AccelerationForce(
    Acceleration=&Multiply_VectorByFloat(
     Vector=<300 0 0>
     Float=&Divide_Float(A=&MaskFloatBySpawnGroup(Masked Float=1, Passthrough Float=-1), B=$System.RibbonHandScalar)
    )
    Coordinate Space=::Simulation
   )
   &Drag(Use Linear Drag=true, Use Rotational Drag=true)
   &SolveForcesAndVelocity(Clamp Velocity=false, Limit Acceleration=false)
   /Game/Effects/Niagara/Modules/Ribbons/FN_UpdateRibbon_v6(Enable Skeletal Position=true, SkeletalMesh:NiagaraDataInterfaceSkeletalMesh=$System.Skeletal Mesh)
   &DynamicMaterialParameters(
    Index 0 Param 1=&MaskFloatBySpawnGroup(Masked Float=0.1, Passthrough Float=0.51)
    Index 0 Param 2=&FloatFromCurve(FloatCurve=DataInterface() {propertyValues {Curve {keys [
         {arriveTangent=-0.3062, interpMode=RCIM_Cubic, leaveTangent=-0.3062, tangentMode=RCTM_User, time=0, value=0.5}
         {arriveTangent=0.0167, interpMode=RCIM_Cubic, leaveTangent=0.0167, tangentMode=RCTM_Break, time=1, value=0}
        ]}}})
    Index 0 Param 3=1
    Index 0 Param 4=1
    Module.Param0WriteEnabled=true
    Module.Param1WriteEnabled=true
    Module.Param2WriteEnabled=true
    Module.Param3WriteEnabled=true
   )
   &ScaleRibbonWidth(Ribbon Width Scale=&FloatFromCurve(
     FloatCurve=DataInterface() {propertyValues {Curve {keys [
         {arriveTangent=-0.0063, interpMode=RCIM_Cubic, leaveTangent=-0.0063, tangentMode=RCTM_User, time=0, value=0.5}
         {arriveTangent=2.6071, interpMode=RCIM_Cubic, leaveTangent=2.6071, time=0.4, value=2.025}
         <at 0.7 2.325>
        ]}}}
     Scale Curve=0.6
    ))
   &ScaleColor(
    Scale Alpha=&Multiply_Float(
     A=&FNTimeOfDayScalar(Night Scale=1.5, Morning Scale=2, Evening Scale=2)
     B=&FloatFromCurve(FloatCurve=DataInterface() {propertyValues {Curve {keys [<at 0 0>,<at 0.1 1>,<at 1 1>]}}})
    )
    Scale RGB=&Multiply_Vector(
     A=&MakeVectorFromLinearColorRGB(Color=<rgba 1 1 1 1>)
     B=&VectorFromFloat(Value=3)
    )
    ScaleA=true
    ScaleRGB=true
   )
  ]
  rendererClasses [&NiagaraRibbonRendererProperties]
  renderers [
   {rendererClass=&NiagaraRibbonRendererProperties}
   {rendererClass=&NiagaraRibbonRendererProperties, rendererIndex=1}
  ]
 }
 RibbonPurple {
  emitterUpdateScript [
   &EmitterState()
   &FN_SpawnRibbon_v3(Number of Ribbons=2, Skeleton Mesh:NiagaraDataInterfaceSkeletalMesh=$System.Skeletal Mesh, SpawnRate=7)
  ]
  particleSpawnScript [
   &FN_SetSpawnGroup(Number of Spawn Groups=2)
   &FN_InitializeRibbon_v5(
    Enable Skeletal Position=true
    Lifetime:NiagaraFloat=&Multiply_Float(A=1.2, B=$System.RibbonHandScalar)
    Ribbon Width=50
    SkeletalMesh:NiagaraDataInterfaceSkeletalMesh=$System.Skeletal Mesh
   )
   &AddVelocity(Velocity=$System.PaintVelocity, Velocity Speed Scale=$System.PaintVelocityScale) {module=AddVelocity001}
   /SetVariables(Particles.Color=&ColorFromCurve(
     ColorCurve=DataInterface() {propertyValues {
       AlphaCurve {keys [<at 0 1 RCIM_Cubic>]}
       BlueCurve {keys [<at 0 0.276>,<at 0.3701 0.086 RCIM_Cubic>]}
       GreenCurve {keys [<at 0 0>,<at 0.3701 0 RCIM_Cubic>]}
       RedCurve {keys [<at 0 0.0615>,<at 0.3701 0.2865 RCIM_Cubic>]}
      }}
     CurveIndex=&RandomRangeFloat(Minimum=0, Maximum=1)
    )) {isSetParameters=true, module=SetVariables_#6}
  ]
  particleUpdateScript [
   &AccelerationForce(
    Acceleration=&Divide_Vector(A:Vector3f=<0 200 0>, B:Vector3f=&VectorFromFloat(Value=$System.RibbonHandScalar))
    Coordinate Space=::Simulation
   )
   &Drag(Use Linear Drag=true, Use Rotational Drag=true)
   &SolveForcesAndVelocity(Clamp Velocity=false, Limit Acceleration=false)
   &FN_UpdateRibbon_v6(Enable Skeletal Position=true, SkeletalMesh:NiagaraDataInterfaceSkeletalMesh=$System.Skeletal Mesh)
   &DynamicMaterialParameters(
    Index 0 Param 1=&MaskFloatBySpawnGroup(Masked Float=0, Passthrough Float=0.5)
    Index 0 Param 2=&FloatFromCurve(FloatCurve=DataInterface() {propertyValues {Curve {keys [
         {arriveTangent=-0.3062, interpMode=RCIM_Cubic, leaveTangent=-0.3062, tangentMode=RCTM_User, time=0, value=1}
         {arriveTangent=0.0167, interpMode=RCIM_Cubic, leaveTangent=0.0167, tangentMode=RCTM_Break, time=1, value=0}
        ]}}})
    Module.Param0WriteEnabled=true
    Module.Param1WriteEnabled=true
    Module.Param2WriteEnabled=false
    Module.Param3WriteEnabled=false
   )
   &ScaleRibbonWidth(Ribbon Width Scale=&FloatFromCurve(
     FloatCurve=DataInterface() {propertyValues {Curve {keys [
         {arriveTangent=-0.0063, interpMode=RCIM_Cubic, leaveTangent=-0.0063, tangentMode=RCTM_User, time=0, value=0.5}
         {arriveTangent=2.6071, interpMode=RCIM_Cubic, leaveTangent=2.6071, time=0.4, value=2.025}
         <at 0.7 2.325>
        ]}}}
     Scale Curve=0.6
    ))
   &ScaleColor(
    Scale Alpha=&Multiply_Float(
     A=&FNTimeOfDayScalar(Night Scale=1.5, Morning Scale=2, Evening Scale=2)
     B=&FloatFromCurve(FloatCurve=DataInterface() {propertyValues {Curve {keys [<at 0 0>,<at 0.1 1>,<at 1 1>]}}})
    )
    Scale RGB=&Multiply_Vector(
     A=&MakeVectorFromLinearColorRGB(Color=<rgba 1 1 1 1>)
     B=&VectorFromFloat(Value=3)
    )
    ScaleA=true
    ScaleRGB=true
   )
  ]
  rendererClasses [&NiagaraRibbonRendererProperties]
  renderers [
   {rendererClass=&NiagaraRibbonRendererProperties}
   {rendererClass=&NiagaraRibbonRendererProperties, rendererIndex=1}
  ]
 }
 RibbonPurpleFeet {
  emitterUpdateScript [
   &EmitterState()
   &FN_SpawnRibbon_v3(Number of Ribbons=2, Skeleton Mesh:NiagaraDataInterfaceSkeletalMesh=$System.SkelMeshFeet, SpawnRate=5)
  ]
  particleSpawnScript [
   &FN_SetSpawnGroup(Number of Spawn Groups=2)
   &FN_InitializeRibbon_v5(
    Enable Skeletal Position=true
    Lifetime:NiagaraFloat=&Multiply_Float(A=0.6, B=$System.RibbonFeetScalar)
    Ribbon Width=40
    SkeletalMesh:NiagaraDataInterfaceSkeletalMesh=$System.SkelMeshFeet
   )
   &AddVelocity(Velocity=$System.PaintVelocity, Velocity Speed Scale=$System.PaintVelocityScale)
   /SetVariables(Particles.Color=&ColorFromCurve(
     ColorCurve=DataInterface() {propertyValues {
       AlphaCurve {keys [<at 0 1 RCIM_Cubic>]}
       BlueCurve {keys [<at 0 0.276>,<at 0.3701 0.086 RCIM_Cubic>]}
       GreenCurve {keys [<at 0 0>,<at 0.3701 0 RCIM_Cubic>]}
       RedCurve {keys [<at 0 0.0615>,<at 0.3701 0.2865 RCIM_Cubic>]}
      }}
     CurveIndex=&RandomRangeFloat(Minimum=0, Maximum=1)
    )) {isSetParameters=true, module=SetVariables_#6}
  ]
  particleUpdateScript [
   &AccelerationForce(
    Acceleration=&Divide_Vector(A:Vector3f=<1 200 1>, B:Vector3f=&VectorFromFloat(Value=$System.RibbonFeetScalar))
    Coordinate Space=::Simulation
   )
   &Drag(Use Linear Drag=true, Use Rotational Drag=true)
   &SolveForcesAndVelocity(Clamp Velocity=false, Limit Acceleration=false)
   &FN_UpdateRibbon_v6(Enable Skeletal Position=true, SkeletalMesh:NiagaraDataInterfaceSkeletalMesh=$System.SkelMeshFeet)
   &DynamicMaterialParameters(
    Index 0 Param 2=&FloatFromCurve(FloatCurve=DataInterface() {propertyValues {Curve {keys [
         {arriveTangent=-0.3062, interpMode=RCIM_Cubic, leaveTangent=-0.3062, tangentMode=RCTM_User, time=0, value=0.5}
         {arriveTangent=0.0167, interpMode=RCIM_Cubic, leaveTangent=0.0167, tangentMode=RCTM_Break, time=1, value=0}
        ]}}})
    Module.Param0WriteEnabled=true
    Module.Param1WriteEnabled=true
    Module.Param2WriteEnabled=true
    Module.Param3WriteEnabled=true
   )
   &ScaleRibbonWidth(Ribbon Width Scale=&FloatFromCurve(
     FloatCurve=DataInterface() {propertyValues {Curve {keys [
         {arriveTangent=-0.0063, interpMode=RCIM_Cubic, leaveTangent=-0.0063, tangentMode=RCTM_User, time=0, value=0.5}
         {arriveTangent=2.6071, interpMode=RCIM_Cubic, leaveTangent=2.6071, time=0.4, value=2.025}
         <at 0.7 2.325>
        ]}}}
     Scale Curve=0.6
    ))
   &ScaleColor(
    Scale Alpha=&Multiply_Float(
     A=&FNTimeOfDayScalar(Night Scale=1.5, Morning Scale=2, Evening Scale=2)
     B=&FloatFromCurve(FloatCurve=DataInterface() {propertyValues {Curve {keys [<at 0 0>,<at 0.1 1>,<at 1 1>]}}})
    )
    Scale RGB=&Multiply_Vector(
     A=&MakeVectorFromLinearColorRGB(Color=<rgba 1 1 1 1>)
     B=&VectorFromFloat(Value=3)
    )
    ScaleA=true
    ScaleRGB=true
   )
   /SetVariables(Particles.BlackRibbonWidth:NiagaraFloat=&Multiply_Float(A=$Particles.RibbonWidth, B=1.15)) {isSetParameters=true}
  ]
  rendererClasses [&NiagaraRibbonRendererProperties, &NiagaraSpriteRendererProperties]
  renderers [
   {rendererClass=&NiagaraRibbonRendererProperties}
   {rendererClass=&NiagaraRibbonRendererProperties, rendererIndex=1}
   {rendererClass=&NiagaraSpriteRendererProperties, rendererIndex=2}
  ]
 }
 RibbonRed {
  emitterUpdateScript [
   &EmitterState()
   &FN_SpawnRibbon_v3(Number of Ribbons=2, Skeleton Mesh:NiagaraDataInterfaceSkeletalMesh=$System.Skeletal Mesh, SpawnRate=7)
  ]
  particleSpawnScript [
   &FN_SetSpawnGroup(Number of Spawn Groups=2)
   &FN_InitializeRibbon_v5(
    Enable Skeletal Position=true
    Lifetime:NiagaraFloat=&Multiply_Float(A=0.78, B=$System.RibbonHandScalar)
    Ribbon Width=50
    SkeletalMesh:NiagaraDataInterfaceSkeletalMesh=$System.Skeletal Mesh
   )
   &AddVelocity(Velocity=$System.PaintVelocity, Velocity Speed Scale=$System.PaintVelocityScale) {module=AddVelocity001}
   /SetVariables(Particles.Color=&ColorFromCurve(
     ColorCurve=DataInterface() {propertyValues {
       AlphaCurve {keys [<at 0 1 RCIM_Cubic>]}
       BlueCurve {keys [<at 0 0>,<at 0.5525 0.0152 RCIM_Cubic>]}
       GreenCurve {keys [<at 0 0.115>,<at 0.5525 0.0137 RCIM_Cubic>]}
       RedCurve {keys [<at 0 0.276>,<at 0.5525 0.3968 RCIM_Cubic>]}
      }}
     CurveIndex=&RandomRangeFloat(Minimum=0, Maximum=1)
    )) {isSetParameters=true, module=SetVariables_#6}
  ]
  particleUpdateScript [
   &AccelerationForce(
    Acceleration=&Divide_Vector(A:Vector3f=<0 -300 0>, B:Vector3f=&VectorFromFloat(Value=$System.RibbonHandScalar))
    Coordinate Space=::Local
   )
   &Drag(Use Linear Drag=true, Use Rotational Drag=true)
   &SolveForcesAndVelocity(Clamp Velocity=false, Limit Acceleration=false)
   &FN_UpdateRibbon_v6(Enable Skeletal Position=true, SkeletalMesh:NiagaraDataInterfaceSkeletalMesh=$System.Skeletal Mesh)
   &DynamicMaterialParameters(
    Index 0 Param 1=&MaskFloatBySpawnGroup(Masked Float=0.3, Passthrough Float=0.53)
    Index 0 Param 2=&FloatFromCurve(FloatCurve=DataInterface() {propertyValues {Curve {keys [
         {arriveTangent=-0.3062, interpMode=RCIM_Cubic, leaveTangent=-0.3062, tangentMode=RCTM_User, time=0, value=0.5}
         {arriveTangent=0.0167, interpMode=RCIM_Cubic, leaveTangent=0.0167, tangentMode=RCTM_Break, time=1, value=0}
        ]}}})
    Index 0 Param 3=1
    Index 0 Param 4=1
    Module.Param0WriteEnabled=true
    Module.Param1WriteEnabled=true
    Module.Param2WriteEnabled=true
    Module.Param3WriteEnabled=true
   )
   &ScaleRibbonWidth(Ribbon Width Scale=&FloatFromCurve(
     FloatCurve=DataInterface() {propertyValues {Curve {keys [
         {arriveTangent=-0.0063, interpMode=RCIM_Cubic, leaveTangent=-0.0063, tangentMode=RCTM_User, time=0, value=0.5}
         {arriveTangent=2.6071, interpMode=RCIM_Cubic, leaveTangent=2.6071, time=0.4, value=2.025}
         <at 0.7 2.325>
        ]}}}
     Scale Curve=0.6
    ))
   &ScaleColor(
    Scale Alpha=&Multiply_Float(
     A=&FNTimeOfDayScalar(Night Scale=1.5, Morning Scale=2, Evening Scale=2)
     B=&FloatFromCurve(FloatCurve=DataInterface() {propertyValues {Curve {keys [<at 0 0>,<at 0.1 1>,<at 1 1>]}}})
    )
    Scale RGB=&Multiply_Vector(
     A=&MakeVectorFromLinearColorRGB(Color=<rgba 1 1 1 1>)
     B=&VectorFromFloat(Value=3)
    )
    ScaleA=true
    ScaleRGB=true
   )
  ]
  rendererClasses [&NiagaraRibbonRendererProperties]
  renderers [
   {rendererClass=&NiagaraRibbonRendererProperties}
   {rendererClass=&NiagaraRibbonRendererProperties, rendererIndex=1}
  ]
 }
 RibbonWhite {
  emitterUpdateScript [
   &EmitterState()
   &FN_SpawnRibbon_v3(Number of Ribbons=2, Skeleton Mesh:NiagaraDataInterfaceSkeletalMesh=$System.Skeletal Mesh, SpawnRate=5)
  ]
  particleSpawnScript [
   &FN_SetSpawnGroup(Number of Spawn Groups=2)
   &FN_InitializeRibbon_v5(
    Enable Skeletal Position=true
    Lifetime:NiagaraFloat=&Multiply_Float(A=0.3, B=$System.RibbonHandScalar)
    Ribbon Width=50
    SkeletalMesh:NiagaraDataInterfaceSkeletalMesh=$System.Skeletal Mesh
   )
   &AddVelocity(Velocity=$System.PaintVelocity, Velocity Speed Scale=$System.PaintVelocityScale) {module=AddVelocity001}
   /SetVariables() {isSetParameters=true, module=SetVariables_#6}
  ]
  particleUpdateScript [
   &AccelerationForce(
    Acceleration=&Divide_Vector(A:Vector3f=<0 200 0>, B:Vector3f=&VectorFromFloat(Value=$System.RibbonHandScalar))
    Coordinate Space=::Local
   )
   &Drag(Use Linear Drag=true, Use Rotational Drag=true)
   &SolveForcesAndVelocity(Clamp Velocity=false, Limit Acceleration=false)
   &FN_UpdateRibbon_v6(Enable Skeletal Position=true, SkeletalMesh:NiagaraDataInterfaceSkeletalMesh=$System.Skeletal Mesh)
   &DynamicMaterialParameters(
    Index 0 Param 1=&MaskFloatBySpawnGroup(Masked Float=0.4, Passthrough Float=0.54)
    Index 0 Param 2=&FloatFromCurve(FloatCurve=DataInterface() {propertyValues {Curve {keys [
         {arriveTangent=-0.3062, interpMode=RCIM_Cubic, leaveTangent=-0.3062, tangentMode=RCTM_User, time=0, value=0.5}
         {arriveTangent=0.0167, interpMode=RCIM_Cubic, leaveTangent=0.0167, tangentMode=RCTM_Break, time=1, value=0}
        ]}}})
    Index 0 Param 3=1
    Index 0 Param 4=1
    Module.Param0WriteEnabled=true
    Module.Param1WriteEnabled=true
    Module.Param2WriteEnabled=true
    Module.Param3WriteEnabled=true
   )
   &ScaleRibbonWidth(Ribbon Width Scale=&FloatFromCurve(
     FloatCurve=DataInterface() {propertyValues {Curve {keys [
         {arriveTangent=-0.0063, interpMode=RCIM_Cubic, leaveTangent=-0.0063, tangentMode=RCTM_User, time=0, value=0.5}
         {arriveTangent=2.6071, interpMode=RCIM_Cubic, leaveTangent=2.6071, time=0.4, value=2.025}
         <at 0.7 2.325>
        ]}}}
     Scale Curve=0.6
    ))
   /SetVariables(Particles.BlackRibbonWidth:NiagaraFloat=&Multiply_Float(A=$Particles.RibbonWidth, B=1.1)) {isSetParameters=true}
   &ScaleColor(
    Scale Alpha=&Multiply_Float(
     A=&FNTimeOfDayScalar(Night Scale=1.5, Morning Scale=2, Evening Scale=2)
     B=&FloatFromCurve(FloatCurve=DataInterface() {propertyValues {Curve {keys [<at 0 0>,<at 0.1 1>,<at 1 1>]}}})
    )
    Scale RGB=&Multiply_Vector(
     A=&MakeVectorFromLinearColorRGB(Color=<rgba 1 1 1 1>)
     B=&VectorFromFloat(Value=3)
    )
    ScaleA=true
    ScaleRGB=true
   )
  ]
  rendererClasses [&NiagaraRibbonRendererProperties]
  renderers [
   {rendererClass=&NiagaraRibbonRendererProperties}
   {rendererClass=&NiagaraRibbonRendererProperties, rendererIndex=1}
  ]
 }
 RibbonYellow {
  emitterUpdateScript [
   &EmitterState()
   &FN_SpawnRibbon_v3(Number of Ribbons=2, Skeleton Mesh:NiagaraDataInterfaceSkeletalMesh=$System.Skeletal Mesh, SpawnRate=7)
  ]
  particleSpawnScript [
   &FN_SetSpawnGroup(Number of Spawn Groups=2)
   &FN_InitializeRibbon_v5(
    Enable Skeletal Position=true
    Lifetime:NiagaraFloat=&Multiply_Float(A=0.84, B=$System.RibbonHandScalar)
    Ribbon Width=50
    SkeletalMesh:NiagaraDataInterfaceSkeletalMesh=$System.Skeletal Mesh
   )
   &AddVelocity(Velocity=$System.PaintVelocity, Velocity Speed Scale=$System.PaintVelocityScale) {module=AddVelocity001}
   /SetVariables(Particles.Color=&ColorFromCurve(
     ColorCurve=DataInterface() {propertyValues {
       AlphaCurve {keys [<at 0 1 RCIM_Cubic>]}
       BlueCurve {keys [<at 0 0>,<at 0.5525 0 RCIM_Cubic>]}
       GreenCurve {keys [<at 0 0.115>,<at 0.5525 0.4118 RCIM_Cubic>]}
       RedCurve {keys [<at 0 0.276>,<at 0.5525 0.6198 RCIM_Cubic>]}
      }}
     CurveIndex=&RandomRangeFloat(Minimum=0, Maximum=1)
    )) {isSetParameters=true, module=SetVariables_#6}
  ]
  particleUpdateScript [
   &AccelerationForce(
    Acceleration=&Multiply_VectorByFloat(
     Vector=<300 0 0>
     Float=&Divide_Float(A=&MaskFloatBySpawnGroup(Masked Float=-1, Passthrough Float=1), B=$System.RibbonHandScalar)
    )
    Coordinate Space=::Simulation
   )
   &Drag(Use Linear Drag=true, Use Rotational Drag=true)
   &SolveForcesAndVelocity(Clamp Velocity=false, Limit Acceleration=false)
   &FN_UpdateRibbon_v6(Enable Skeletal Position=true, SkeletalMesh:NiagaraDataInterfaceSkeletalMesh=$System.Skeletal Mesh)
   &DynamicMaterialParameters(
    Index 0 Param 1=&MaskFloatBySpawnGroup(Masked Float=0.2, Passthrough Float=0.52)
    Index 0 Param 2=&FloatFromCurve(FloatCurve=DataInterface() {propertyValues {Curve {keys [
         {arriveTangent=-0.3062, interpMode=RCIM_Cubic, leaveTangent=-0.3062, tangentMode=RCTM_User, time=0, value=0.5}
         {arriveTangent=0.0167, interpMode=RCIM_Cubic, leaveTangent=0.0167, tangentMode=RCTM_Break, time=1, value=0}
        ]}}})
    Index 0 Param 3=1
    Index 0 Param 4=1
    Module.Param0WriteEnabled=true
    Module.Param1WriteEnabled=true
    Module.Param2WriteEnabled=true
    Module.Param3WriteEnabled=true
   )
   &ScaleRibbonWidth(Ribbon Width Scale=&FloatFromCurve(
     FloatCurve=DataInterface() {propertyValues {Curve {keys [
         {arriveTangent=-0.0063, interpMode=RCIM_Cubic, leaveTangent=-0.0063, tangentMode=RCTM_User, time=0, value=0.5}
         {arriveTangent=2.6071, interpMode=RCIM_Cubic, leaveTangent=2.6071, time=0.4, value=2.025}
         <at 0.7 2.325>
        ]}}}
     Scale Curve=0.6
    ))
   &ScaleColor(
    Scale Alpha=&Multiply_Float(
     A=&FNTimeOfDayScalar(Night Scale=1.5, Morning Scale=2, Evening Scale=2)
     B=&FloatFromCurve(FloatCurve=DataInterface() {propertyValues {Curve {keys [<at 0 0>,<at 0.1 1>,<at 1 1>]}}})
    )
    Scale RGB=&Multiply_Vector(
     A=&MakeVectorFromLinearColorRGB(Color=<rgba 1 1 1 1>)
     B=&VectorFromFloat(Value=3)
    )
    ScaleA=true
    ScaleRGB=true
   )
  ]
  rendererClasses [&NiagaraRibbonRendererProperties]
  renderers [
   {rendererClass=&NiagaraRibbonRendererProperties}
   {rendererClass=&NiagaraRibbonRendererProperties, rendererIndex=1}
  ]
 }
}
summary {
 systemName=NS_Contrail_UnderGroundRebel
 emitters [
  {emitterName=RibbonPurple, rendererClasses [&NiagaraRibbonRendererProperties]}
  {emitterName=RibbonBlue, rendererClasses [&NiagaraRibbonRendererProperties]}
  {emitterName=RibbonYellow, rendererClasses [&NiagaraRibbonRendererProperties]}
  {emitterName=RibbonRed, rendererClasses [&NiagaraRibbonRendererProperties]}
  {emitterName=RibbonWhite, rendererClasses [&NiagaraRibbonRendererProperties]}
  {
   emitterName=RibbonPurpleFeet
   rendererClasses [&NiagaraRibbonRendererProperties, &NiagaraSpriteRendererProperties]
  }
  {emitterName=DropletsHands, simTarget=GPUComputeSim, rendererClasses [&NiagaraSpriteRendererProperties]}
  {emitterName=DropletsFeet, simTarget=GPUComputeSim, rendererClasses [&NiagaraSpriteRendererProperties]}
  {bEnabled=false, emitterName=FortniteLetters, rendererClasses [&NiagaraSpriteRendererProperties]}
 ]
 userVariables [
  User.bIsFrontEnd {type=NiagaraBool}
  User.bIsFrontEndPreview {type=NiagaraBool}
  User.ParaGlideLeanAlpha {type=NiagaraFloat, defaultValue=0, description=""}
  User.PlayerSpeed {type=NiagaraFloat, defaultValue=0, description=""}
  User.RibbonIDOffset {type=NiagaraInt32}
  User.SkeletalMesh {
   type=NiagaraDataInterfaceSkeletalMesh
   defaultValue {
    struct=NiagaraExt_VariableValue_DataInterface
    value {dataInterface=&self:NiagaraDataInterfaceSkeletalMesh_6}
   }
  }
 ]
}
